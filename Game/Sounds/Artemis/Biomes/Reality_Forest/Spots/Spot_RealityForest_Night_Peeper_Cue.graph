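# Sound cue for one-shot alien peeper ambience at night in the Reality Forest biome, using bird one-shot attenuation and the ambient one-shots sound class.
# A quality-level node sends its first two slots to a random pick of eight AlienPeep waves and its last two to a pool of four.
# Each pool is equally weighted and passes through its own modulator for slight pitch and volume variation.
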SoundCue /Game/Sounds/Artemis/Biomes/Reality_Forest/Spots/Spot_RealityForest_Night_Peeper_Cue
root=QualityLevel_0
nodes [
 {id=Modulator_1, children [Random_0], properties {pitchMax=1.1, volumeMax=1.1, volumeMin=0.9}}
 {id=Modulator_2, children [Random_1], properties {pitchMax=1.1, volumeMax=1.1, volumeMin=0.9}}
 {id=QualityLevel_0, children [Modulator_1,Modulator_1,Modulator_2,Modulator_2]}
 {
  id=Random_0
  children [WavePlayer_7, WavePlayer_14, WavePlayer_15, WavePlayer_16, WavePlayer_17, WavePlayer_18, WavePlayer_19, WavePlayer_20]
  properties {weights [1,1,1,1,1,1,1,1]}
 }
 {id=Random_1, children [WavePlayer_21, WavePlayer_22, WavePlayer_23, WavePlayer_24], properties {weights [1,1,1,1]}}
 {id=WavePlayer_14, properties {soundWaveAssetPtr=&/AMB_Biome_Reality_Forest_SS_AlienPeep_02}}
 {id=WavePlayer_15, properties {soundWaveAssetPtr=&/AMB_Biome_Reality_Forest_SS_AlienPeep_03}}
 {id=WavePlayer_16, properties {soundWaveAssetPtr=&/AMB_Biome_Reality_Forest_SS_AlienPeep_04}}
 {id=WavePlayer_17, properties {soundWaveAssetPtr=&/AMB_Biome_Reality_Forest_SS_AlienPeep_05}}
 {id=WavePlayer_18, properties {soundWaveAssetPtr=&/AMB_Biome_Reality_Forest_SS_AlienPeep_06}}
 {id=WavePlayer_19, properties {soundWaveAssetPtr=&/AMB_Biome_Reality_Forest_SS_AlienPeep_07}}
 {id=WavePlayer_20, properties {soundWaveAssetPtr=&/AMB_Biome_Reality_Forest_SS_AlienPeep_08}}
 {id=WavePlayer_21, properties {soundWaveAssetPtr=&/AMB_Biome_Reality_Forest_SS_AlienPeep_01}}
 {id=WavePlayer_22, properties {soundWaveAssetPtr=&AMB_Biome_Reality_Forest_SS_AlienPeep_02}}
 {id=WavePlayer_23, properties {soundWaveAssetPtr=&AMB_Biome_Reality_Forest_SS_AlienPeep_03}}
 {id=WavePlayer_24, properties {soundWaveAssetPtr=&AMB_Biome_Reality_Forest_SS_AlienPeep_04}}
 {id=WavePlayer_7, properties {soundWaveAssetPtr=&AMB_Biome_Reality_Forest_SS_AlienPeep_01}}
]
properties {
 attenuationSettings=&Fort_TileAmbient_OneShot_Bird_Att
 bEnableBusSends=false
 maxDistance=8005
 soundClassObject=&Fort_Ambient_OneShots
 volumeMultiplier=0.376
 preEffectBusSends [{}]
}
