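# Looping ambient spot cue of root and foliage sounds for the open Reality Bloom in the Reality Forest biome.
# Each mixer layers a continuous random loop with a randomly delayed loop, both varied in pitch and volume.
# A quality switch gives the first two levels an eleven-wave pool and the last two a four-wave subset.
# Overrides add occlusion with low-pass filtering and limit concurrent instances.
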SoundCue /Game/Sounds/Artemis/Biomes/Reality_Forest/Spots/Spot_RealityForest_RealityBloom_Open_Loop_Cue
root=QualityLevel_0
nodes [
 {id=Delay_0, children [Modulator_3], properties {delayMax=5, delayMin=1}}
 {id=Delay_1, children [Modulator_2], properties {delayMax=5, delayMin=1}}
 {id=Looping_0, children [Modulator_0]}
 {id=Looping_1, children [Delay_0]}
 {id=Looping_2, children [Modulator_1]}
 {id=Looping_3, children [Delay_1]}
 {id=Mixer_0, children [Looping_2,Looping_3], properties {inputVolume [1,1]}}
 {id=Mixer_1, children [Looping_0,Looping_1], properties {inputVolume [1,1]}}
 {id=Modulator_0, children [Random_2], properties {pitchMin=0.85, volumeMax=1.25, volumeMin=0.7}}
 {id=Modulator_1, children [Random_4], properties {pitchMin=0.85, volumeMax=1.25, volumeMin=0.7}}
 {id=Modulator_2, children [Random_6], properties {pitchMin=0.85, volumeMax=1.25, volumeMin=0.7}}
 {id=Modulator_3, children [Random_3], properties {pitchMin=0.85, volumeMax=1.25, volumeMin=0.7}}
 {id=QualityLevel_0, children [Mixer_1,Mixer_1,Mixer_0,Mixer_0]}
 {
  id=Random_2
  children [
   WavePlayer_10
   WavePlayer_11
   WavePlayer_12
   WavePlayer_13
   WavePlayer_14
   WavePlayer_15
   WavePlayer_16
   WavePlayer_17
   WavePlayer_18
   WavePlayer_19
   WavePlayer_20
  ]
  properties {weights [1,1,1,1,1,1,1,1,1,1,1]}
 }
 {
  id=Random_3
  children [
   WavePlayer_21
   WavePlayer_34
   WavePlayer_35
   WavePlayer_36
   WavePlayer_37
   WavePlayer_38
   WavePlayer_39
   WavePlayer_40
   WavePlayer_41
   WavePlayer_42
   WavePlayer_43
  ]
  properties {weights [1,1,1,1,1,1,1,1,1,1,1]}
 }
 {id=Random_4, children [WavePlayer_3,WavePlayer_4,WavePlayer_5,WavePlayer_6], properties {weights [1,1,1,1]}}
 {id=Random_6, children [WavePlayer_26, WavePlayer_27, WavePlayer_28, WavePlayer_29], properties {weights [1,1,1,1]}}
 {id=WavePlayer_10, properties {soundWaveAssetPtr=&/AMB_Reality_Forest_Root_Foliage_SS_07}}
 {id=WavePlayer_11, properties {soundWaveAssetPtr=&/AMB_Reality_Forest_Root_Foliage_SS_10}}
 {id=WavePlayer_12, properties {soundWaveAssetPtr=&/AMB_Reality_Forest_Root_Foliage_SS_09}}
 {id=WavePlayer_13, properties {soundWaveAssetPtr=&/AMB_Reality_Forest_Root_Foliage_SS_11}}
 {id=WavePlayer_14, properties {soundWaveAssetPtr=&/AMB_Reality_Forest_Root_Foliage_SS_02}}
 {id=WavePlayer_15, properties {soundWaveAssetPtr=&/AMB_Reality_Forest_Root_Foliage_SS_01}}
 {id=WavePlayer_16, properties {soundWaveAssetPtr=&/AMB_Reality_Forest_Root_Foliage_SS_03}}
 {id=WavePlayer_17, properties {soundWaveAssetPtr=&/AMB_Reality_Forest_Root_Foliage_SS_04}}
 {id=WavePlayer_18, properties {soundWaveAssetPtr=&/AMB_Reality_Forest_Root_Foliage_SS_06}}
 {id=WavePlayer_19, properties {soundWaveAssetPtr=&/AMB_Reality_Forest_Root_Foliage_SS_05}}
 {id=WavePlayer_20, properties {soundWaveAssetPtr=&/AMB_Reality_Forest_Root_Foliage_SS_08}}
 {id=WavePlayer_21, properties {soundWaveAssetPtr=&AMB_Reality_Forest_Root_Foliage_SS_07}}
 {id=WavePlayer_26, properties {soundWaveAssetPtr=&AMB_Reality_Forest_Root_Foliage_SS_11}}
 {id=WavePlayer_27, properties {soundWaveAssetPtr=&AMB_Reality_Forest_Root_Foliage_SS_02}}
 {id=WavePlayer_28, properties {soundWaveAssetPtr=&AMB_Reality_Forest_Root_Foliage_SS_01}}
 {id=WavePlayer_29, properties {soundWaveAssetPtr=&AMB_Reality_Forest_Root_Foliage_SS_03}}
 {id=WavePlayer_3, properties {soundWaveAssetPtr=&AMB_Reality_Forest_Root_Foliage_SS_11}}
 {id=WavePlayer_34, properties {soundWaveAssetPtr=&AMB_Reality_Forest_Root_Foliage_SS_10}}
 {id=WavePlayer_35, properties {soundWaveAssetPtr=&AMB_Reality_Forest_Root_Foliage_SS_09}}
 {id=WavePlayer_36, properties {soundWaveAssetPtr=&AMB_Reality_Forest_Root_Foliage_SS_11}}
 {id=WavePlayer_37, properties {soundWaveAssetPtr=&AMB_Reality_Forest_Root_Foliage_SS_02}}
 {id=WavePlayer_38, properties {soundWaveAssetPtr=&AMB_Reality_Forest_Root_Foliage_SS_01}}
 {id=WavePlayer_39, properties {soundWaveAssetPtr=&AMB_Reality_Forest_Root_Foliage_SS_03}}
 {id=WavePlayer_4, properties {soundWaveAssetPtr=&AMB_Reality_Forest_Root_Foliage_SS_02}}
 {id=WavePlayer_40, properties {soundWaveAssetPtr=&AMB_Reality_Forest_Root_Foliage_SS_04}}
 {id=WavePlayer_41, properties {soundWaveAssetPtr=&AMB_Reality_Forest_Root_Foliage_SS_06}}
 {id=WavePlayer_42, properties {soundWaveAssetPtr=&AMB_Reality_Forest_Root_Foliage_SS_05}}
 {id=WavePlayer_43, properties {soundWaveAssetPtr=&AMB_Reality_Forest_Root_Foliage_SS_08}}
 {id=WavePlayer_5, properties {soundWaveAssetPtr=&AMB_Reality_Forest_Root_Foliage_SS_01}}
 {id=WavePlayer_6, properties {soundWaveAssetPtr=&AMB_Reality_Forest_Root_Foliage_SS_03}}
]
properties {
 attenuationSettings=/Game/Sounds/Attenuation/Fort_TileAmbient_OneShot_Wider_Att
 bEnableBusSends=false
 bOverrideAttenuation=true
 bOverrideConcurrency=true
 maxDistance=2400
 soundClassObject=&Fort_Ambient_OneShots
 volumeMultiplier=0.501
 attenuationOverrides {
  bEnableOcclusion=true
  falloffDistance=2000
  occlusionLowPassFilterFrequency=2000
  occlusionTraceChannel=ECC_GameTraceChannel11
  occlusionVolumeAttenuation=0.5
 }
 concurrencyOverrides {maxCount=2}
 preEffectBusSends [{}]
}
