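# Sound cue for tonal wind one-shot spot ambience in the Reality Forest biome, using a wide tile-ambient attenuation and the ambient one-shot sound class.
# A quality-level node maps its first two slots to a random pick of eight wind waves and its last two to a smaller pool of four.
# Each pool feeds through its own modulator, which adds slight pitch and volume variation to each play.
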SoundCue /Game/Sounds/Artemis/Biomes/Reality_Forest/Spots/Spot_RealityForest_Wind_03_Cue
root=QualityLevel_0
nodes [
 {id=Modulator_1, children [Random_2], properties {pitchMax=1.1, volumeMax=1.1, volumeMin=0.9}}
 {id=Modulator_2, children [Random_3], properties {pitchMax=1.1, volumeMax=1.1, volumeMin=0.9}}
 {id=QualityLevel_0, children [Modulator_1,Modulator_1,Modulator_2,Modulator_2]}
 {
  id=Random_2
  children [WavePlayer_10, WavePlayer_11, WavePlayer_12, WavePlayer_13, WavePlayer_14, WavePlayer_15, WavePlayer_16, WavePlayer_17]
  properties {weights [1,1,1,1,1,1,1,1]}
 }
 {id=Random_3, children [WavePlayer_18, WavePlayer_19, WavePlayer_20, WavePlayer_21], properties {weights [1,1,1,1]}}
 {id=WavePlayer_10, properties {soundWaveAssetPtr=&/AMB_Biome_Reality_Forest_SS_Wind_Tonal_08}}
 {id=WavePlayer_11, properties {soundWaveAssetPtr=&/AMB_Biome_Reality_Forest_SS_Wind_Tonal_01}}
 {id=WavePlayer_12, properties {soundWaveAssetPtr=&/AMB_Biome_Reality_Forest_SS_Wind_Tonal_02}}
 {id=WavePlayer_13, properties {soundWaveAssetPtr=&/AMB_Biome_Reality_Forest_SS_Wind_Tonal_03}}
 {id=WavePlayer_14, properties {soundWaveAssetPtr=&/AMB_Biome_Reality_Forest_SS_Wind_Tonal_04}}
 {id=WavePlayer_15, properties {soundWaveAssetPtr=&/AMB_Biome_Reality_Forest_SS_Wind_Tonal_05}}
 {id=WavePlayer_16, properties {soundWaveAssetPtr=&/AMB_Biome_Reality_Forest_SS_Wind_Tonal_06}}
 {id=WavePlayer_17, properties {soundWaveAssetPtr=&/AMB_Biome_Reality_Forest_SS_Wind_Tonal_07}}
 {id=WavePlayer_18, properties {soundWaveAssetPtr=&AMB_Biome_Reality_Forest_SS_Wind_Tonal_08}}
 {id=WavePlayer_19, properties {soundWaveAssetPtr=&AMB_Biome_Reality_Forest_SS_Wind_Tonal_01}}
 {id=WavePlayer_20, properties {soundWaveAssetPtr=&AMB_Biome_Reality_Forest_SS_Wind_Tonal_02}}
 {id=WavePlayer_21, properties {soundWaveAssetPtr=&AMB_Biome_Reality_Forest_SS_Wind_Tonal_03}}
]
properties {
 attenuationSettings=/Game/Sounds/Attenuation/Fort_TileAmbient_OneShot_Wider_Att
 bEnableBusSends=false
 maxDistance=8000
 soundClassObject=&Fort_Ambient_OneShots
 volumeMultiplier=0.401
 preEffectBusSends [{}]
}
